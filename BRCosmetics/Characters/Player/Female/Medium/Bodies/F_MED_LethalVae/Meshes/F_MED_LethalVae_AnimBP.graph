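# Post-process style AnimBP for the LethalVae female body cosmetic, animating skirt, buckles and bat gem over the parent mesh's pose.
# Copied pose runs through a deform Control Rig, then weighted constraints and spring AnimDynamics drive those bones.
# Crouch, skydive/jump and thigh yaw blend in additive buckle and skirt offsets.
# Both leg-back values read thigh_r, and skydiveGravity is set but unused by the AnimGraph.
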AnimBlueprint /BRCosmetics/Characters/Player/Female/Medium/Bodies/F_MED_LethalVae/Meshes/F_MED_LethalVae_AnimBP
target_skeleton=&/F_MED_LethalVae_Skeleton
graphs [
 AnimGraph {nodes [
   n1 = Control Rig(Source=Copy Pose From Mesh() {bUseAttachedParent=true}) {
    alphaBoolBlend {blendOption=Linear}
    controlRigAssetReference {blueprintRigClass=&/F_MED_DeformRig_LethalVae_ControlRig.F_MED_DeformRig_LethalVae_ControlRig_C}
   }
   n2 = "Constraint - master_L_skirt_side_jnt (P)"(ComponentPose=Local To Component(LocalPose=@n1), Alpha=1, ConstraintWeights_0=1) {
    alphaBoolBlend {blendOption=Linear}
    boneToModify {boneName=master_L_skirt_side_jnt}
    constraintSetup [{targetBone {boneName=thigh_l}}]
    constraintWeights [1]
   }
   n3 = "Constraint - master_L_skirt_bt_jnt (PP)"!(ComponentPose=@n2, Alpha=1, ConstraintWeights_0=0.5, ConstraintWeights_1:float=0.5) -> (Pose) {
    alphaBoolBlend {blendOption=Linear}
    boneToModify {boneName=master_L_skirt_bt_jnt}
    constraintSetup [{targetBone {boneName=thigh_l}}, {targetBone {boneName=pelvis}}]
    constraintWeights [1,1]
   }
   n4 = "Constraint - master_L_skirt_back_jnt (PP)"!(ComponentPose=@n3, Alpha=1, ConstraintWeights_0=0.2, ConstraintWeights_1:float=0.8) -> (Pose) {
    alphaBoolBlend {blendOption=Linear}
    boneToModify {boneName=master_L_skirt_back_jnt}
    constraintSetup [{targetBone {boneName=thigh_l}}, {targetBone {boneName=pelvis}}]
    constraintWeights [1,1]
   }
   n5 = Anim Dynamics - Bone: dyn_L_skirt_back_1_jnt!(ComponentPose=@n4, Alpha=1, GravityOverride:Vector="0,0,-500", ExternalForce="0,0,0") -> (Pose) {
    angularDampingOverride=1
    angularSpringConstant=250
    bAngularSpring=true
    bOverrideAngularDamping=true
    bUseGravityOverride=true
    gravityScale=0.35
    simulationSpace=RootRelative
    alphaBoolBlend {blendOption=Linear}
    boundBone {boneName=dyn_L_skirt_back_1_jnt}
    componentAppliedLinearAccClamp=<3000 3000 3000>
    componentLinearAccScale=<0.5 0.4 0.4>
    componentLinearVelScale=<0.5 0.4 0.4>
    gravityOverride {z=-500}
    physicsBodyDefinitions [{
      boundBone {boneName=dyn_L_skirt_back_1_jnt}
      boxExtents=<30 50 50>
      constraintSetup {angularLimitsMax=<10 5 20>, angularLimitsMin=<-10 -58 -2>, angularTarget=<1 0 0>}
      localJointOffset=<15 0 0>
     }]
    relativeSpaceBone {boneName=spine_02}
   }
   n6 = Anim Dynamics - Bone: dyn_L_skirt_bt_1_jnt!(ComponentPose=@n5, Alpha=1, GravityOverride:Vector="0,0,-500", ExternalForce="0,0,0") -> (Pose) {
    angularDampingOverride=1
    angularSpringConstant=250
    bAngularSpring=true
    bOverrideAngularDamping=true
    bUseGravityOverride=true
    gravityScale=0.35
    simulationSpace=RootRelative
    alphaBoolBlend {blendOption=Linear}
    boundBone {boneName=dyn_L_skirt_bt_1_jnt}
    componentAppliedLinearAccClamp=<3000 3000 3000>
    componentLinearAccScale=<0.5 0.4 0.4>
    componentLinearVelScale=<0.5 0.4 0.4>
    gravityOverride {z=-500}
    physicsBodyDefinitions [{
      boundBone {boneName=dyn_L_skirt_bt_1_jnt}
      boxExtents=<38 50 50>
      constraintSetup {angularLimitsMax=<10 5 20>, angularLimitsMin=<-10 -58 -2>, angularTarget=<1 0 0>}
      localJointOffset=<18 0 0>
     }]
    relativeSpaceBone {boneName=spine_02}
   }
   n7 = Anim Dynamics - Bone: dyn_L_skirt_side_1_jnt(ComponentPose=@n6, Alpha=1, ExternalForce="0,0,0") {
    angularDampingOverride=1
    angularSpringConstant=250
    bAngularSpring=true
    bOverrideAngularDamping=true
    bUseGravityOverride=true
    gravityScale=0.35
    simulationSpace=RootRelative
    alphaBoolBlend {blendOption=Linear}
    boundBone {boneName=dyn_L_skirt_side_1_jnt}
    componentAppliedLinearAccClamp=<3000 3000 3000>
    componentLinearAccScale=<0.5 0.4 0.4>
    componentLinearVelScale=<0.5 0.4 0.4>
    gravityOverride {z=-500}
    physicsBodyDefinitions [{
      boundBone {boneName=dyn_L_skirt_side_1_jnt}
      boxExtents=<45 50 50>
      constraintSetup {angularLimitsMax=<10 5 20>, angularLimitsMin=<-10 -58 -2>, angularTarget=<1 0 0>}
      localJointOffset=<20 0 0>
     }]
    relativeSpaceBone {boneName=spine_02}
   }
   n8 = "Constraint - master_R_skirt_side_jnt (P)"(ComponentPose=@n7, Alpha=1, ConstraintWeights_0=1) {
    alphaBoolBlend {blendOption=Linear}
    boneToModify {boneName=master_R_skirt_side_jnt}
    constraintSetup [{targetBone {boneName=thigh_r}}]
    constraintWeights [1]
   }
   n9 = "Constraint - master_R_skirt_bt_jnt (PP)"!(ComponentPose=@n8, Alpha=1, ConstraintWeights_0=0.5, ConstraintWeights_1:float=0.5) -> (Pose) {
    alphaBoolBlend {blendOption=Linear}
    boneToModify {boneName=master_R_skirt_bt_jnt}
    constraintSetup [{targetBone {boneName=thigh_r}}, {targetBone {boneName=pelvis}}]
    constraintWeights [1,1]
   }
   n10 = "Constraint - master_R_skirt_back_jnt (PP)"!(ComponentPose=@n9, Alpha=1, ConstraintWeights_0=0.2, ConstraintWeights_1:float=0.8) -> (Pose) {
    alphaBoolBlend {blendOption=Linear}
    boneToModify {boneName=master_R_skirt_back_jnt}
    constraintSetup [{targetBone {boneName=thigh_r}}, {targetBone {boneName=pelvis}}]
    constraintWeights [1,1]
   }
   n11 = Anim Dynamics - Bone: dyn_R_skirt_back_1_jnt!(ComponentPose=@n10, Alpha=1, GravityOverride:Vector="0,0,-500", ExternalForce="0,0,0") -> (Pose) {
    angularDampingOverride=1
    angularSpringConstant=250
    bAngularSpring=true
    bOverrideAngularDamping=true
    bUseGravityOverride=true
    gravityScale=0.35
    simulationSpace=RootRelative
    alphaBoolBlend {blendOption=Linear}
    boundBone {boneName=dyn_R_skirt_back_1_jnt}
    componentAppliedLinearAccClamp=<3000 3000 3000>
    componentLinearAccScale=<0.5 0.4 0.4>
    componentLinearVelScale=<0.5 0.4 0.4>
    gravityOverride {z=-500}
    physicsBodyDefinitions [{
      boundBone {boneName=dyn_R_skirt_back_1_jnt}
      boxExtents=<30 50 50>
      constraintSetup {angularLimitsMax=<10 5 20>, angularLimitsMin=<-10 -58 -2>, angularTarget=<1 0 0>}
      localJointOffset=<-15 0 0>
     }]
    relativeSpaceBone {boneName=spine_02}
   }
   n12 = Anim Dynamics - Bone: dyn_R_skirt_bt_1_jnt!(ComponentPose=@n11, Alpha=1, GravityOverride:Vector="0,0,-500", ExternalForce="0,0,0") -> (Pose) {
    angularDampingOverride=1
    angularSpringConstant=250
    bAngularSpring=true
    bOverrideAngularDamping=true
    bUseGravityOverride=true
    gravityScale=0.35
    simulationSpace=RootRelative
    alphaBoolBlend {blendOption=Linear}
    boundBone {boneName=dyn_R_skirt_bt_1_jnt}
    componentAppliedLinearAccClamp=<3000 3000 3000>
    componentLinearAccScale=<0.5 0.4 0.4>
    componentLinearVelScale=<0.5 0.4 0.4>
    gravityOverride {z=-500}
    physicsBodyDefinitions [{
      boundBone {boneName=dyn_R_skirt_bt_1_jnt}
      boxExtents=<38 50 50>
      constraintSetup {angularLimitsMax=<10 5 20>, angularLimitsMin=<-10 -58 -2>, angularTarget=<1 0 0>}
      localJointOffset=<-18 0 0>
     }]
    relativeSpaceBone {boneName=spine_02}
   }
   n13 = Anim Dynamics - Bone: dyn_R_skirt_side_1_jnt(ComponentPose=@n12, Alpha=1, ExternalForce="0,0,0") {
    angularDampingOverride=1
    angularSpringConstant=250
    bAngularSpring=true
    bOverrideAngularDamping=true
    bUseGravityOverride=true
    gravityScale=0.35
    simulationSpace=RootRelative
    alphaBoolBlend {blendOption=Linear}
    boundBone {boneName=dyn_R_skirt_side_1_jnt}
    componentAppliedLinearAccClamp=<3000 3000 3000>
    componentLinearAccScale=<0.5 0.4 0.4>
    componentLinearVelScale=<0.5 0.4 0.4>
    gravityOverride {z=-500}
    physicsBodyDefinitions [{
      boundBone {boneName=dyn_R_skirt_side_1_jnt}
      boxExtents=<45 50 50>
      constraintSetup {angularLimitsMax=<10 5 20>, angularLimitsMin=<-10 -58 -2>, angularTarget=<1 0 0>}
      localJointOffset=<-20 0 0>
     }]
    relativeSpaceBone {boneName=spine_02}
   }
   n14 = "Constraint - lower_L_buckle_jnt (PPPPP)"!(
    ComponentPose=@n13
    Alpha=1
    ConstraintWeights_0=0.4288
    ConstraintWeights_1:float=0.0149
    ConstraintWeights_2:float=0.2557
    ConstraintWeights_3:float=0.1731
    ConstraintWeights_4:float=0.1274
   ) -> (Pose) {
    alphaBoolBlend {blendOption=Linear}
    boneToModify {boneName=lower_L_buckle_jnt}
    constraintSetup [
     {targetBone {boneName=pelvis}}
     {targetBone {boneName=thigh_twist_01_l}}
     {targetBone {boneName=thigh_l}}
     {targetBone {boneName=deform_groin_l}}
     {targetBone {boneName=dyn_L_skirt_side_1_jnt}}
    ]
    constraintWeights [1,1,1,1,1]
   }
   n15 = "Constraint - lower_R_buckle_jnt (PPPPP)"!(
    ComponentPose=@n14
    Alpha=1
    ConstraintWeights_0=0.4442
    ConstraintWeights_1:float=0.0114
    ConstraintWeights_2:float=0.2391
    ConstraintWeights_3:float=0.1668
    ConstraintWeights_4:float=0.1385
   ) -> (Pose) {
    alphaBoolBlend {blendOption=Linear}
    boneToModify {boneName=lower_R_buckle_jnt}
    constraintSetup [
     {targetBone {boneName=pelvis}}
     {targetBone {boneName=thigh_twist_01_r}}
     {targetBone {boneName=thigh_r}}
     {targetBone {boneName=deform_groin_r}}
     {targetBone {boneName=dyn_R_skirt_side_1_jnt}}
    ]
    constraintWeights [1,1,1,1,1]
   }
   n16 = "Constraint - upper_R_buckle_jnt (PPPPP)"!(
    ComponentPose=@n15
    Alpha=1
    ConstraintWeights_0=0.1278
    ConstraintWeights_1:float=0.0277
    ConstraintWeights_2:float=0.0741
    ConstraintWeights_3:float=0.7469
    ConstraintWeights_4:float=0.0235
   ) -> (Pose) {
    alphaBoolBlend {blendOption=Linear}
    boneToModify {boneName=upper_R_buckle_jnt}
    constraintSetup [
     {targetBone {boneName=spine_02}}
     {targetBone {boneName=spine_04}}
     {targetBone {boneName=spine_01}}
     {targetBone {boneName=pelvis}}
     {targetBone {boneName=deform_groin_r}}
    ]
    constraintWeights [1,1,1,1,1]
   }
   n17 = "Constraint - upper_L_buckle_jnt (PPPPP)"!(
    ComponentPose=@n16
    Alpha=1
    ConstraintWeights_0=0.1728
    ConstraintWeights_1:float=0.0425
    ConstraintWeights_2:float=0.0996
    ConstraintWeights_3:float=0.6614
    ConstraintWeights_4:float=0.0238
   ) -> (Pose) {
    alphaBoolBlend {blendOption=Linear}
    boneToModify {boneName=upper_L_buckle_jnt}
    constraintSetup [
     {targetBone {boneName=spine_02}}
     {targetBone {boneName=spine_04}}
     {targetBone {boneName=spine_01}}
     {targetBone {boneName=pelvis}}
     {targetBone {boneName=deform_groin_l}}
    ]
    constraintWeights [1,1,1,1,1]
   }
   n18 = Get Crouching() -> (Crouching:double)
   n19 = "Transform (Modify) Bone - Bone: upper_L_buckle_jnt"(ComponentPose=@n17, Alpha=@n18, Translation="0,-0.7,2", Rotation="1,0,0", Scale="1,1,1") {
    rotationMode=BMM_Additive
    rotationSpace=BCS_BoneSpace
    translationMode=BMM_Additive
    translationSpace=BCS_BoneSpace
    alphaBoolBlend {blendOption=Linear}
    alphaScaleBiasClamp {bInterpResult=true, interpSpeedIncreasing=5}
    boneToModify {boneName=upper_L_buckle_jnt}
    rotation {pitch=1}
    translation {y=-0.7, z=2}
   }
   n20 = "Transform (Modify) Bone - Bone: upper_R_buckle_jnt"(ComponentPose=@n19, Alpha=@n18, Translation="-1,0,-1.5", Rotation="0,0,0", Scale="1,1,1") {
    rotationMode=BMM_Additive
    rotationSpace=BCS_BoneSpace
    translationMode=BMM_Additive
    translationSpace=BCS_BoneSpace
    alphaBoolBlend {blendOption=Linear}
    alphaScaleBiasClamp {bInterpResult=true, interpSpeedIncreasing=5}
    boneToModify {boneName=upper_R_buckle_jnt}
    translation {x=-1, z=-1.5}
   }
   n21 = "Transform (Modify) Bone - Bone: lower_L_buckle_jnt"(ComponentPose=@n20, Alpha=@n18, Translation="-2,4.6,1", Rotation="-25,15,65", Scale="1,1,1") {
    rotationMode=BMM_Additive
    rotationSpace=BCS_BoneSpace
    translationMode=BMM_Additive
    translationSpace=BCS_BoneSpace
    alphaBoolBlend {blendOption=Linear}
    alphaScaleBiasClamp {bInterpResult=true, interpSpeedIncreasing=5}
    boneToModify {boneName=lower_L_buckle_jnt}
    rotation=<rot -25 15 65>
    translation=<-2 4.6 1>
   }
   n22 = "Constraint - dyn_bat_gem_jnt (PP)"!(ComponentPose=@n21, Alpha=1, ConstraintWeights_0=0.0929, ConstraintWeights_1:float=0.9071) -> (Pose) {
    alphaBoolBlend {blendOption=Linear}
    boneToModify {boneName=dyn_bat_gem_jnt}
    constraintSetup [{targetBone {boneName=spine_04}}, {targetBone {boneName=spine_05}}]
    constraintWeights [1,1]
   }
   n23 = Anim Dynamics - Bone: dyn_bat_gem_jnt(ComponentPose=@n22, Alpha=1, ExternalForce="0,0,0") {
    angularDampingOverride=0.99
    angularSpringConstant=20000
    bAngularSpring=true
    bOverrideAngularDamping=true
    bUseGravityOverride=true
    simulationSpace=RootRelative
    alphaBoolBlend {blendOption=Linear}
    boundBone {boneName=dyn_bat_gem_jnt}
    componentAppliedLinearAccClamp=<3000 3000 3000>
    componentLinearAccScale=<0.3 0.3 0.3>
    componentLinearVelScale=<0.3 0.3 0.3>
    gravityOverride {z=-980}
    physicsBodyDefinitions [{
      boundBone {boneName=dyn_bat_gem_jnt}
      boxExtents=<10 10 10>
      constraintSetup {angularLimitsMax=<0 0 10>, angularLimitsMin=<0 -20 -10>, angularTarget=<0 0 0>}
      localJointOffset=<2 0 0>
     }]
   }
   n24 = Get Skydiving or Jumping() -> (Skydiving or Jumping:double)
   n25 = "Transform (Modify) Bone - Bone: master_L_skirt_back_jnt"(ComponentPose=@n23, Alpha=@n24, Translation="0,0,0", Rotation="20,0,0", Scale="1,1,1") {
    rotationMode=BMM_Additive
    rotationSpace=BCS_BoneSpace
    alphaBoolBlend {blendOption=Linear}
    alphaScaleBiasClamp {bInterpResult=true, interpSpeedIncreasing=5}
    boneToModify {boneName=master_L_skirt_back_jnt}
    rotation {pitch=20}
   }
   n26 = "Transform (Modify) Bone - Bone: master_R_skirt_back_jnt"(ComponentPose=@n25, Alpha=@n24, Translation="0,0,0", Rotation="20,0,0", Scale="1,1,1") {
    rotationMode=BMM_Additive
    rotationSpace=BCS_BoneSpace
    alphaBoolBlend {blendOption=Linear}
    alphaScaleBiasClamp {bInterpResult=true, interpSpeedIncreasing=5}
    boneToModify {boneName=master_R_skirt_back_jnt}
    rotation {pitch=20}
   }
   n27 = "Transform (Modify) Bone - Bone: master_L_skirt_bt_jnt"(ComponentPose=@n26, Alpha=@n24, Translation="0,0,0", Rotation="10,-10,0", Scale="1,1,1") {
    rotationMode=BMM_Additive
    rotationSpace=BCS_BoneSpace
    alphaBoolBlend {blendOption=Linear}
    alphaScaleBiasClamp {bInterpResult=true, interpSpeedIncreasing=5}
    boneToModify {boneName=master_L_skirt_bt_jnt}
    rotation {pitch=10, yaw=-10}
   }
   n28 = "Transform (Modify) Bone - Bone: master_R_skirt_bt_jnt"(ComponentPose=@n27, Alpha=@n24, Translation="0,0,0", Rotation="10,-10,0", Scale="1,1,1") {
    rotationMode=BMM_Additive
    rotationSpace=BCS_BoneSpace
    alphaBoolBlend {blendOption=Linear}
    alphaScaleBiasClamp {bInterpResult=true, interpSpeedIncreasing=5}
    boneToModify {boneName=master_R_skirt_bt_jnt}
    rotation {pitch=10, yaw=-10}
   }
   n29 = "Transform (Modify) Bone - Bone: master_L_skirt_back_jnt"(ComponentPose=@n28, Alpha=Get L leg back() -> (L leg back:double), Translation="0,0,0", Rotation="20,0,0", Scale="1,1,1") {
    rotationMode=BMM_Additive
    rotationSpace=BCS_BoneSpace
    alphaBoolBlend {blendOption=Linear}
    alphaScaleBiasClamp {bInterpResult=true, interpSpeedIncreasing=5}
    boneToModify {boneName=master_L_skirt_back_jnt}
    rotation {pitch=20}
   }
   n30 = "Transform (Modify) Bone - Bone: master_R_skirt_back_jnt"(ComponentPose=@n29, Alpha=Get R leg back() -> (R leg back:double), Translation="0,0,0", Rotation="20,0,0", Scale="1,1,1") {
    rotationMode=BMM_Additive
    rotationSpace=BCS_BoneSpace
    alphaBoolBlend {blendOption=Linear}
    alphaScaleBiasClamp {bInterpResult=true, interpSpeedIncreasing=5}
    boneToModify {boneName=master_R_skirt_back_jnt}
    rotation {pitch=20}
   }
   Output Pose(Result=Component To Local(ComponentPose=@n30)) {layerGroup=None}
  ]}
 EventGraph {nodes [
   Event BlueprintUpdateAnimation()
   > Set Crouching(Crouching:double="To Float (Boolean)"(InBool=Get bIsCrouching())) -> (Output_Get:double)
   n1 = GetOwnerSkeletalMeshComponent()
   n2 = BreakTransform(InTransform=GetDeltaTransformFromRefPose(self=@n1, BoneName=thigh_r, BaseName=pelvis))
   > Set L leg back(L leg back:double=MapRangeClamped(Value=BreakRotator(InRot=@n2.Rotation).Yaw, InRangeA=8, InRangeB=15, OutRangeA=0.6)) -> (Output_Get:double)
   n3 = BreakTransform(InTransform=GetDeltaTransformFromRefPose(self=@n1, BoneName=thigh_r, BaseName=pelvis))
   > Set R leg back(R leg back:double=MapRangeClamped(Value=BreakRotator(InRot=@n3.Rotation).Yaw, InRangeA=8, InRangeB=15, OutRangeA=0.6)) -> (Output_Get:double)
   n4 = Get bIsSkyDiving()
   n5 = Get bIsJumping()
   > Set Skydiving or Jumping(Skydiving or Jumping:double="To Float (Boolean)"(InBool=OR Boolean(A=@n5, B=@n4))) -> (Output_Get:double)
   > Set skydiveGravity(skydiveGravity:Vector=SelectVector(A="0,0,-980", B="0,0,-500", bPickA=Get bIsSkyDiving())) -> (Output_Get:Vector)
  ]}
]
variables [
 Crouching {type=number}
 L leg back {type=number}
 R leg back {type=number}
 skydiveGravity {type=Vector}
 Skydiving or Jumping {type=number}
]
