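# Physics asset for the LethalVae female character's backbling/face accessory part, covering dynamic hair, fringe and shoulder-piece bones.
# Capsules on upper spine, clavicles, neck and head form the body shapes.
# Three-joint main, left, right and side hair chains, plus single-joint fringe and shoulder pieces, hang from master bones with near-locked twist.
# Masters link back to the head, and free-angular constraints cross-link lower hair joints to each other, clavicles and spine.
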PhysicsAsset /BRCosmetics/Characters/Player/Female/Medium/Bodies/F_MED_LethalVae/Meshes/Parts/F_MED_LethalVae_Backbling_FaceAcc_Physics
bodies {
 clavicle_l [{
   shapeType=Capsule
   length=18.5
   radius=5.5
   shapeName=clavicle_l_capsule
   center=<9.8248 -3.1032 2.392>
   rotation=<rot 1.7277 -80.1489 80.1489>
  }]
 clavicle_r [{
   shapeType=Capsule
   length=15.5
   radius=5
   shapeName=clavicle_r_capsule
   center=<-10.2479 3.0746 -2.382>
   rotation=<rot -1.7286 -80.1489 79.851>
  }]
 dyn_L_hair_1_jnt [{shapeType=Capsule, length=4, radius=4, shapeName=dyn_L_hair_1_jnt_capsule, center=<-2.5819 -4 0>}]
 dyn_L_hair_2_jnt [{shapeType=Capsule, length=4, radius=4, shapeName=dyn_L_hair_2_jnt_capsule, center=<-4.5219 -4.0811 0>}]
 dyn_L_hair_3_jnt [{shapeType=Capsule, length=4, radius=5, shapeName=dyn_L_hair_3_jnt_capsule, center=<-6.6228 -3.1873 0>}]
 dyn_L_hair_side_1_jnt [{shapeType=Capsule, length=2, radius=1.5, shapeName=dyn_L_hair_side_1_jnt_capsule, center=<3.6584 1.4016 0>, rotation=<rot 90 0 -10.0004>}]
 dyn_L_hair_side_2_jnt [{
   shapeType=Capsule
   length=2
   radius=1.5
   shapeName=dyn_L_hair_side_2_jnt_capsule
   center=<3.6584 1.4016 0>
   rotation=<rot 90 178.2101 168.2097>
  }]
 dyn_L_hair_side_3_jnt [{shapeType=Capsule, length=5, radius=1.5, shapeName=dyn_L_hair_side_3_jnt_capsule, center=<3.6584 1.4016 0>, rotation=<rot 90 0 -10.0004>}]
 dyn_L_shoulder_piece_jnt [{shapeType=Capsule, length=3.5, radius=1.5, shapeName=dyn_L_shoulder_piece_jnt_capsule, center=<3.171 0 0.6901>, rotation=<rot -70 0.0002 0>}]
 dyn_R_hair_1_jnt [{shapeType=Capsule, length=4, radius=4, shapeName=dyn_R_hair_1_jnt_capsule, center=<2.9363 4 0>}]
 dyn_R_hair_2_jnt [{shapeType=Capsule, length=4, radius=4, shapeName=dyn_R_hair_2_jnt_capsule, center=<4.5553 3.8074 0>}]
 dyn_R_hair_3_jnt [{shapeType=Capsule, length=4, radius=5, shapeName=dyn_R_hair_3_jnt_capsule, center=<6.5236 2.8545 0>}]
 dyn_R_hair_side_1_jnt [{shapeType=Capsule, length=2, radius=1.5, shapeName=dyn_R_hair_side_1_jnt_capsule, center=<-3.9607 -1.7865 0>, rotation=<rot 90 0 -10.0004>}]
 dyn_R_hair_side_2_jnt [{shapeType=Capsule, length=2, radius=1.5, shapeName=dyn_R_hair_side_2_jnt_capsule, center=<-3.9607 -1.7865 0>, rotation=<rot 90 0 -10.0004>}]
 dyn_R_hair_side_3_jnt [{shapeType=Capsule, length=5, radius=1.5, shapeName=dyn_R_hair_side_3_jnt_capsule, center=<-3.9607 -1.7865 0>, rotation=<rot 90 0 -10.0004>}]
 dyn_R_shoulder_piece_jnt [{
   shapeType=Capsule
   length=6
   radius=2
   shapeName=dyn_R_shoulder_piece_jnt_capsule
   center=<-4.5856 0 -0.465>
   rotation=<rot -79.9999 0.0005 -0.0004>
  }]
 dyn_fringe_strand_jnt [{
   shapeType=Capsule
   length=6.5
   radius=1
   shapeName=dyn_fringe_strand_jnt_capsule
   center=<4.06 1.0324 0>
   rotation=<rot 90 179.9999 179.9998>
  }]
 dyn_main_hair_1_jnt [{shapeType=Capsule, length=4, radius=4, shapeName=dyn_main_hair_1_jnt_capsule, center=<2.9363 4 0>}]
 dyn_main_hair_2_jnt [{shapeType=Capsule, length=4, radius=4, shapeName=dyn_main_hair_2_jnt_capsule, center=<4.5553 3.8 0>}]
 dyn_main_hair_3_jnt [{shapeType=Capsule, length=4, radius=5, shapeName=dyn_main_hair_3_jnt_capsule, center=<6.5236 2.3309 0>}]
 head [{shapeType=Capsule, length=0.5, radius=6.5, shapeName=head_capsule, center=<4.3675 4.2212 0>, rotation=<rot 90 0 0>}]
 master_L_hair_jnt [{shapeType=Capsule, length=1, radius=1, shapeName=master_L_hair_jnt_capsule}]
 master_L_hair_side_jnt [{shapeType=Capsule, length=1, radius=1, shapeName=master_L_hair_side_jnt_capsule}]
 master_L_shoulder_piece_jnt [{shapeType=Capsule, length=0.505, radius=0.505, shapeName=master_L_shoulder_piece_jnt_capsule, rotation=<rot 0 -90 0>}]
 master_R_hair_jnt [{shapeType=Capsule, length=1, radius=1, shapeName=master_R_hair_jnt_capsule}]
 master_R_hair_side_jnt [{shapeType=Capsule, length=1, radius=1, shapeName=master_R_hair_side_jnt_capsule}]
 master_R_shoulder_piece_jnt [{shapeType=Capsule, length=0.505, radius=0.505, shapeName=master_R_shoulder_piece_jnt_capsule, rotation=<rot 0 -90 0>}]
 master_fringe_strand_jnt [{shapeType=Capsule, length=1, radius=1, shapeName=master_fringe_strand_jnt_capsule}]
 master_main_hair_jnt [{shapeType=Capsule, length=1, radius=1, shapeName=master_main_hair_jnt_capsule}]
 neck_01 [{shapeType=Capsule, length=8.7, radius=4.5, shapeName=neck_01_capsule, center=<2.6 -1.2175 0>, rotation=<rot 90 -180 -179.9999>}]
 spine_04 [{shapeType=Capsule, length=17, radius=6, shapeName=spine_05_capsule, center=<2.7519 -2.0011 0>}]
 spine_05 [
  {shapeType=Capsule, length=23.5, radius=6.5, shapeName=spine_05_capsule, center=<1.8077 -1.5594 -0.1282>, rotation=<rot 0 -0.0003 10>}
  {shapeType=Capsule, length=24, radius=6, shapeName=spine_05_capsule, center=<2.0234 3.2009 0>}
  {shapeType=Capsule, length=25.5, radius=4.5, shapeName=spine_05_capsule, center=<12.7203 -1.1223 0>}
 ]
}
constraints [
 {
  bone1Name=clavicle_l
  bone2Name=spine_05
  swing1LimitDegrees=45
  swing1Motion=Free
  swing2LimitDegrees=45
  swing2Motion=Free
  twistLimitDegrees=45
  twistMotion=Free
 }
 {
  bone1Name=clavicle_r
  bone2Name=spine_05
  swing1LimitDegrees=45
  swing1Motion=Free
  swing2LimitDegrees=45
  swing2Motion=Free
  twistLimitDegrees=45
  twistMotion=Free
 }
 {
  bone1Name=neck_01
  bone2Name=spine_05
  swing1LimitDegrees=45
  swing1Motion=Free
  swing2LimitDegrees=45
  swing2Motion=Free
  twistLimitDegrees=45
  twistMotion=Free
 }
 {
  bone1Name=head
  bone2Name=neck_01
  swing1LimitDegrees=45
  swing1Motion=Free
  swing2LimitDegrees=45
  swing2Motion=Free
  twistLimitDegrees=45
  twistMotion=Free
 }
 {bone1Name=dyn_main_hair_1_jnt, bone2Name=master_main_hair_jnt, swing1LimitDegrees=2, swing2LimitDegrees=13, twistLimitDegrees=1}
 {bone1Name=dyn_main_hair_2_jnt, bone2Name=dyn_main_hair_1_jnt, swing1LimitDegrees=2, swing2LimitDegrees=13, twistLimitDegrees=1}
 {bone1Name=dyn_main_hair_3_jnt, bone2Name=dyn_main_hair_2_jnt, swing1LimitDegrees=2, swing2LimitDegrees=13, twistLimitDegrees=1}
 {bone1Name=dyn_L_hair_side_1_jnt, bone2Name=master_L_hair_side_jnt, swing1LimitDegrees=45, swing2LimitDegrees=15, twistLimitDegrees=1}
 {bone1Name=dyn_L_hair_side_2_jnt, bone2Name=dyn_L_hair_side_1_jnt, swing1LimitDegrees=45, swing2LimitDegrees=25, twistLimitDegrees=1}
 {bone1Name=dyn_L_hair_side_3_jnt, bone2Name=dyn_L_hair_side_2_jnt, swing1LimitDegrees=45, swing2LimitDegrees=45, twistLimitDegrees=1}
 {bone1Name=dyn_R_hair_side_1_jnt, bone2Name=master_R_hair_side_jnt, swing1LimitDegrees=45, swing2LimitDegrees=15, twistLimitDegrees=1}
 {bone1Name=dyn_R_hair_side_2_jnt, bone2Name=dyn_R_hair_side_1_jnt, swing1LimitDegrees=45, swing2LimitDegrees=25, twistLimitDegrees=1}
 {bone1Name=dyn_R_hair_side_3_jnt, bone2Name=dyn_R_hair_side_2_jnt, swing1LimitDegrees=45, swing2LimitDegrees=45, twistLimitDegrees=1}
 {bone1Name=dyn_fringe_strand_jnt, bone2Name=master_fringe_strand_jnt, swing1LimitDegrees=85, swing2LimitDegrees=12, twistLimitDegrees=1}
 {bone1Name=dyn_L_hair_1_jnt, bone2Name=master_L_hair_jnt, swing1LimitDegrees=2, swing2LimitDegrees=13, twistLimitDegrees=1}
 {bone1Name=dyn_L_hair_2_jnt, bone2Name=dyn_L_hair_1_jnt, swing1LimitDegrees=2, swing2LimitDegrees=13, twistLimitDegrees=1}
 {bone1Name=dyn_R_hair_1_jnt, bone2Name=master_R_hair_jnt, swing1LimitDegrees=2, swing2LimitDegrees=13, twistLimitDegrees=1}
 {bone1Name=dyn_R_hair_2_jnt, bone2Name=dyn_R_hair_1_jnt, swing1LimitDegrees=2, swing2LimitDegrees=13, twistLimitDegrees=1}
 {
  bone1Name=spine_05
  bone2Name=spine_04
  swing1LimitDegrees=45
  swing1Motion=Free
  swing2LimitDegrees=45
  swing2Motion=Free
  twistLimitDegrees=45
  twistMotion=Free
 }
 {
  bone1Name=master_main_hair_jnt
  bone2Name=head
  swing1LimitDegrees=45
  swing1Motion=Free
  swing2LimitDegrees=45
  swing2Motion=Free
  twistLimitDegrees=45
  twistMotion=Free
 }
 {
  bone1Name=master_L_hair_jnt
  bone2Name=master_main_hair_jnt
  swing1LimitDegrees=45
  swing1Motion=Free
  swing2LimitDegrees=45
  swing2Motion=Free
  twistLimitDegrees=45
  twistMotion=Free
 }
 {
  bone1Name=master_R_hair_jnt
  bone2Name=master_main_hair_jnt
  swing1LimitDegrees=45
  swing1Motion=Free
  swing2LimitDegrees=45
  swing2Motion=Free
  twistLimitDegrees=45
  twistMotion=Free
 }
 {
  bone1Name=master_L_hair_side_jnt
  bone2Name=head
  swing1LimitDegrees=45
  swing1Motion=Free
  swing2LimitDegrees=45
  swing2Motion=Free
  twistLimitDegrees=45
  twistMotion=Free
 }
 {
  bone1Name=master_R_hair_side_jnt
  bone2Name=head
  swing1LimitDegrees=45
  swing1Motion=Free
  swing2LimitDegrees=45
  swing2Motion=Free
  twistLimitDegrees=45
  twistMotion=Free
 }
 {
  bone1Name=master_fringe_strand_jnt
  bone2Name=head
  swing1LimitDegrees=45
  swing1Motion=Free
  swing2LimitDegrees=45
  swing2Motion=Free
  twistLimitDegrees=45
  twistMotion=Free
 }
 {bone1Name=dyn_L_shoulder_piece_jnt, bone2Name=master_L_shoulder_piece_jnt, swing1LimitDegrees=25, swing2LimitDegrees=5, twistLimitDegrees=1}
 {bone1Name=dyn_R_shoulder_piece_jnt, bone2Name=master_R_shoulder_piece_jnt, swing1LimitDegrees=25, swing2LimitDegrees=5, twistLimitDegrees=1}
 {
  bone1Name=master_L_shoulder_piece_jnt
  bone2Name=head
  swing1LimitDegrees=45
  swing1Motion=Free
  swing2LimitDegrees=45
  swing2Motion=Free
  twistLimitDegrees=45
  twistMotion=Free
 }
 {
  bone1Name=master_R_shoulder_piece_jnt
  bone2Name=head
  swing1LimitDegrees=45
  swing1Motion=Free
  swing2LimitDegrees=45
  swing2Motion=Free
  twistLimitDegrees=45
  twistMotion=Free
 }
 {
  bone1Name=dyn_L_hair_3_jnt
  bone2Name=dyn_main_hair_3_jnt
  swing1LimitDegrees=45
  swing1Motion=Free
  swing2LimitDegrees=45
  swing2Motion=Free
  twistLimitDegrees=45
  twistMotion=Free
 }
 {
  bone1Name=dyn_R_hair_3_jnt
  bone2Name=dyn_main_hair_3_jnt
  swing1LimitDegrees=45
  swing1Motion=Free
  swing2LimitDegrees=45
  swing2Motion=Free
  twistLimitDegrees=45
  twistMotion=Free
 }
 {bone1Name=dyn_L_hair_3_jnt, bone2Name=dyn_L_hair_2_jnt, swing1LimitDegrees=2, swing2LimitDegrees=13, twistLimitDegrees=1}
 {bone1Name=dyn_R_hair_3_jnt, bone2Name=dyn_R_hair_2_jnt, swing1LimitDegrees=2, swing2LimitDegrees=13, twistLimitDegrees=1}
 {
  bone1Name=dyn_R_hair_3_jnt
  bone2Name=clavicle_r
  swing1LimitDegrees=45
  swing1Motion=Free
  swing2LimitDegrees=45
  swing2Motion=Free
  twistLimitDegrees=45
  twistMotion=Free
 }
 {
  bone1Name=dyn_L_hair_3_jnt
  bone2Name=clavicle_l
  swing1LimitDegrees=45
  swing1Motion=Free
  swing2LimitDegrees=45
  swing2Motion=Free
  twistLimitDegrees=45
  twistMotion=Free
 }
 {
  bone1Name=dyn_L_hair_2_jnt
  bone2Name=clavicle_l
  swing1LimitDegrees=45
  swing1Motion=Free
  swing2LimitDegrees=45
  swing2Motion=Free
  twistLimitDegrees=45
  twistMotion=Free
 }
 {
  bone1Name=dyn_R_hair_2_jnt
  bone2Name=spine_05
  swing1LimitDegrees=45
  swing1Motion=Free
  swing2LimitDegrees=45
  swing2Motion=Free
  twistLimitDegrees=45
  twistMotion=Free
 }
 {
  bone1Name=dyn_L_hair_2_jnt
  bone2Name=spine_05
  swing1LimitDegrees=45
  swing1Motion=Free
  swing2LimitDegrees=45
  swing2Motion=Free
  twistLimitDegrees=45
  twistMotion=Free
 }
 {
  bone1Name=dyn_R_hair_2_jnt
  bone2Name=clavicle_r
  swing1LimitDegrees=45
  swing1Motion=Free
  swing2LimitDegrees=45
  swing2Motion=Free
  twistLimitDegrees=45
  twistMotion=Free
 }
]
