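# Post-process-style AnimBlueprint for the LethalVae face accessory that animates its hair with RigidBody physics on a pose copied from the parent mesh.
# Corrective bone rotations driven by head rotation, crouching, parachuting and rail grinding feed two gameplay simulations.
# Wearing a backbling or cape selects a damped simulation with hair constrained to spine and clavicles.
# Front end uses its own physics asset plus fixed bone offsets.
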AnimBlueprint /BRCosmetics/Characters/Player/Female/Medium/Bodies/F_MED_LethalVae/Meshes/Parts/F_MED_LethalVae_FaceAcc_AnimBP
target_skeleton=&/F_MED_LethalVae_FaceAcc_Skeleton
graphs [
 AnimGraph {nodes [
   n1 = RigidBody!(
    ComponentPose=Local To Component(LocalPose=Use cached pose 'TransformPose'())
    Alpha
    OverrideWorldGravity:Vector=Get Gravity Override()
    ExternalForce
   ) -> (Pose) {
    bClampLinearTranslationLimitToRefPose=true
    bOverrideWorldGravity=true
    overridePhysicsAsset=&/F_MED_LethalVae_FaceAcc_Physics
    simulationSpace=BaseBoneSpace
    alphaBoolBlend {blendOption=Linear}
    baseBoneRef {boneName=head}
    componentAppliedLinearAccClamp=<3000 3000 3000>
    componentLinearAccScale=<0.4 0.2 0.2>
    componentLinearVelScale=<0.4 0.2 0.2>
    overrideWorldGravity {z=-980}
    simSpaceSettings {maxAngularAcceleration=1000, maxAngularVelocity=3, maxLinearAcceleration=250, maxLinearVelocity=500, velocityScaleZ=0, worldAlpha=1}
   }
   n2 = Component To Local(ComponentPose=@n1)
   n3 = RigidBody(ComponentPose=Local To Component(LocalPose=Use cached pose 'TransformPose'()), OverrideWorldGravity=Get Gravity Override()) {
    bClampLinearTranslationLimitToRefPose=true
    bOverrideWorldGravity=true
    overridePhysicsAsset=&/F_MED_LethalVae_Backbling_FaceAcc_Physics
    simulationSpace=BaseBoneSpace
    alphaBoolBlend {blendOption=Linear}
    baseBoneRef {boneName=head}
    componentAppliedLinearAccClamp=<3000 3000 3000>
    componentLinearAccScale=<0.1 0.1 0.1>
    componentLinearVelScale=<0.1 0.1 0.1>
    overrideWorldGravity {z=-980}
    simSpaceSettings {maxAngularAcceleration=1000, maxAngularVelocity=3, maxLinearAcceleration=250, maxLinearVelocity=500, velocityScaleZ=0, worldAlpha=1}
   }
   n4 = "Constraint - dyn_main_hair_3_jnt (P)"(ComponentPose=Local To Component(LocalPose=Component To Local(ComponentPose=@n3)), Alpha=0.2, ConstraintWeights_0=1) {
    alphaBoolBlend {blendOption=Linear}
    boneToModify {boneName=dyn_main_hair_3_jnt}
    constraintSetup [{targetBone {boneName=spine_05}}]
    constraintWeights [1]
   }
   n5 = "Constraint - dyn_L_hair_3_jnt (PP)"!(ComponentPose=@n4, Alpha=0.2, ConstraintWeights_0=0.2, ConstraintWeights_1:float=0.8) -> (Pose) {
    alphaBoolBlend {blendOption=Linear}
    boneToModify {boneName=dyn_L_hair_3_jnt}
    constraintSetup [{targetBone {boneName=spine_05}}, {targetBone {boneName=clavicle_l}}]
    constraintWeights [1,1]
   }
   n6 = "Constraint - dyn_R_hair_3_jnt (PP)"!(ComponentPose=@n5, Alpha=0.2, ConstraintWeights_0=0.8, ConstraintWeights_1:float=0.2) -> (Pose) {
    alphaBoolBlend {blendOption=Linear}
    boneToModify {boneName=dyn_R_hair_3_jnt}
    constraintSetup [{targetBone {boneName=spine_05}}, {targetBone {boneName=clavicle_r}}]
    constraintWeights [1,1]
   }
   n7 = "Constraint - dyn_main_hair_2_jnt (P)"(ComponentPose=@n6, Alpha=0.5, ConstraintWeights_0=1) {
    alphaBoolBlend {blendOption=Linear}
    boneToModify {boneName=dyn_main_hair_2_jnt}
    constraintSetup [{targetBone {boneName=spine_05}}]
    constraintWeights [1]
   }
   n8 = "Constraint - dyn_L_hair_2_jnt (PP)"!(ComponentPose=@n7, Alpha=0.5, ConstraintWeights_0=0.2, ConstraintWeights_1:float=0.8) -> (Pose) {
    alphaBoolBlend {blendOption=Linear}
    boneToModify {boneName=dyn_L_hair_2_jnt}
    constraintSetup [{targetBone {boneName=spine_05}}, {targetBone {boneName=clavicle_l}}]
    constraintWeights [1,1]
   }
   n9 = "Constraint - dyn_R_hair_2_jnt (PP)"!(ComponentPose=@n8, Alpha=0.7, ConstraintWeights_0=0.8, ConstraintWeights_1:float=0.2) -> (Pose) {
    alphaBoolBlend {blendOption=Linear}
    boneToModify {boneName=dyn_R_hair_2_jnt}
    constraintSetup [{targetBone {boneName=spine_05}}, {targetBone {boneName=clavicle_r}}]
    constraintWeights [1,1]
   }
   n10 = Component To Local(ComponentPose=@n9)
   n11 = OR Boolean(A=Get Is Wearing Backbling(), B=Get Is Wearing Cape() -> (Is Wearing Cape:bool))
   n12 = "Constraint - master_main_hair_jnt (P)"(ComponentPose=Local To Component(LocalPose=Use cached pose 'StartPose'()), Alpha=1, ConstraintWeights_0=1) {
    alphaBoolBlend {blendOption=Linear}
    boneToModify {boneName=master_main_hair_jnt}
    constraintSetup [{targetBone {boneName=spine_05}}]
    constraintWeights [1]
   }
   n13 = RigidBody(ComponentPose=@n12, OverrideWorldGravity=Get Gravity Override()) {
    bClampLinearTranslationLimitToRefPose=true
    bOverrideWorldGravity=true
    overridePhysicsAsset=&/F_MED_LethalVae_FaceAcc_FrontEnd_Physics
    simulationSpace=BaseBoneSpace
    alphaBoolBlend {blendOption=Linear}
    baseBoneRef {boneName=head}
    componentAppliedLinearAccClamp=<3000 3000 3000>
    componentLinearAccScale=<0.4 0.2 0.2>
    componentLinearVelScale=<0.4 0.2 0.2>
    overrideWorldGravity {z=-980}
    simSpaceSettings {maxAngularAcceleration=1000, maxAngularVelocity=3, maxLinearAcceleration=250, maxLinearVelocity=500, velocityScaleZ=0, worldAlpha=1}
   }
   n14 = "Transform (Modify) Bone - Bone: master_R_shoulder_piece_jnt"(ComponentPose=@n13, Alpha=1, Translation="0,0,0", Rotation="-10,0,0", Scale="1,1,1") {
    rotationMode=BMM_Additive
    rotationSpace=BCS_BoneSpace
    translationSpace=BCS_WorldSpace
    alphaBoolBlend {blendOption=Linear}
    boneToModify {boneName=master_R_shoulder_piece_jnt}
    rotation {pitch=-10}
   }
   n15 = "Transform (Modify) Bone - Bone: dyn_L_hair_3_jnt"(ComponentPose=@n14, Alpha=1, Translation="-0.994,1.294,3.040", Rotation="-46.544,-19.375,29.914", Scale="1,1,1") {
    rotationMode=BMM_Additive
    rotationSpace=BCS_BoneSpace
    translationMode=BMM_Additive
    translationSpace=BCS_WorldSpace
    alphaBoolBlend {blendOption=Linear}
    boneToModify {boneName=dyn_L_hair_3_jnt}
    rotation=<rot -46.544 -19.3751 29.914>
    translation=<-0.994 1.294 3.04>
   }
   n16 = "Transform (Modify) Bone - Bone: dyn_L_hair_2_jnt"!(ComponentPose=@n15, bAlphaBoolEnabled:bool=True, Translation="0,0,0", Rotation="0,0,0", Scale="0.85,1,1") -> (Pose) {
    alphaInputType=Bool
    rotationMode=BMM_Additive
    rotationSpace=BCS_BoneSpace
    scaleMode=BMM_Replace
    translationSpace=BCS_WorldSpace
    alphaBoolBlend {blendOption=Linear}
    boneToModify {boneName=dyn_L_hair_2_jnt}
    scale {x=0.85}
   }
   n17 = "Transform (Modify) Bone - Bone: dyn_main_hair_2_jnt"!(ComponentPose=@n16, bAlphaBoolEnabled:bool=True, Translation="0,0,0", Rotation="2.241,-3.203,20.325", Scale="0.8,1,1") -> (Pose) {
    alphaInputType=Bool
    rotationMode=BMM_Additive
    rotationSpace=BCS_BoneSpace
    scaleMode=BMM_Replace
    translationSpace=BCS_WorldSpace
    alphaBoolBlend {blendOption=Linear}
    boneToModify {boneName=dyn_main_hair_2_jnt}
    rotation=<rot 2.241 -3.2031 20.325>
    scale {x=0.8}
   }
   n18 = "Transform (Modify) Bone - Bone: dyn_R_hair_2_jnt"!(ComponentPose=@n17, bAlphaBoolEnabled:bool=True, Translation="0,0,0", Rotation="2.653,-5.636,20.153", Scale="0.85,1,1") -> (Pose) {
    alphaInputType=Bool
    rotationMode=BMM_Additive
    rotationSpace=BCS_BoneSpace
    scaleMode=BMM_Replace
    translationSpace=BCS_WorldSpace
    alphaBoolBlend {blendOption=Linear}
    boneToModify {boneName=dyn_R_hair_2_jnt}
    rotation=<rot 2.6532 -5.6362 20.153>
    scale {x=0.85}
   }
   n19 = "Transform (Modify) Bone - Bone: dyn_L_hair_2_jnt"(ComponentPose=@n18, bAlphaBoolEnabled=True, Translation="-0.468,2.152,1.610", Rotation="13.005,10.360,-2.940", Scale="0.85,1,1") {
    alphaInputType=Bool
    rotationMode=BMM_Additive
    rotationSpace=BCS_BoneSpace
    scaleMode=BMM_Replace
    translationMode=BMM_Additive
    translationSpace=BCS_BoneSpace
    alphaBoolBlend {blendOption=Linear}
    boneToModify {boneName=dyn_L_hair_2_jnt}
    rotation=<rot 13.005 10.36 -2.94>
    scale {x=0.85}
    translation=<-0.468 2.152 1.61>
   }
   Output Pose(Result=Blend Poses by bool(
     BlendPose_0=Component To Local(ComponentPose=@n19)
     BlendPose_1=Blend Poses by bool(BlendPose_0=@n10, BlendPose_1=@n2, bActiveValue=@n11) {blendType=Linear}
     bActiveValue=Get Front End() -> (Front End:bool)
    ) {blendType=Linear}) {layerGroup=None}
   n20 = "Transform (Modify) Bone - Bone: master_main_hair_jnt"(
    ComponentPose=Local To Component(LocalPose=Use cached pose 'StartPose'())
    Alpha=Get Head Back Rot() -> (Head Back Rot:double)
    Translation="0,0,0"
    Rotation="22,19,-17"
    Scale="1,1,1"
   ) {
    rotationMode=BMM_Additive
    rotationSpace=BCS_BoneSpace
    translationMode=BMM_Additive
    translationSpace=BCS_BoneSpace
    alphaBoolBlend {blendOption=Linear}
    boneToModify {boneName=master_main_hair_jnt}
    rotation=<rot 22 19 -17>
   }
   n21 = "Transform (Modify) Bone - Bone: master_R_hair_jnt"(ComponentPose=@n20, Alpha=Get Head R Rot() -> (Head R Rot:double), Translation="0,0,0", Rotation="0,130,0", Scale="1,1,1") {
    rotationMode=BMM_Additive
    rotationSpace=BCS_BoneSpace
    translationMode=BMM_Additive
    translationSpace=BCS_BoneSpace
    alphaBoolBlend {blendOption=Linear}
    boneToModify {boneName=master_R_hair_jnt}
    rotation {yaw=130}
   }
   n22 = Get Skydive Hanging() -> (Skydive Hanging:double)
   n23 = "Transform (Modify) Bone - Bone: master_R_hair_jnt"(ComponentPose=@n21, Alpha=@n22, Translation="0,0,0", Rotation="0,20,0", Scale="1,1,1") {
    rotationMode=BMM_Additive
    rotationSpace=BCS_BoneSpace
    alphaBoolBlend {blendOption=Linear}
    alphaScaleBiasClamp {interpSpeedDecreasing=5, interpSpeedIncreasing=8}
    boneToModify {boneName=master_R_hair_jnt}
    rotation {yaw=20}
   }
   n24 = "Transform (Modify) Bone - Bone: master_L_hair_jnt"(ComponentPose=@n23, Alpha=@n22, Translation="0,0,0", Rotation="0,40,20", Scale="1,1,1") {
    rotationMode=BMM_Additive
    rotationSpace=BCS_BoneSpace
    alphaBoolBlend {blendOption=Linear}
    alphaScaleBiasClamp {interpSpeedDecreasing=5, interpSpeedIncreasing=8}
    boneToModify {boneName=master_L_hair_jnt}
    rotation {roll=20, yaw=40}
   }
   n25 = "Transform (Modify) Bone - Bone: master_main_hair_jnt"(ComponentPose=@n24, Alpha=@n22, Translation="0,0,0", Rotation="0,20,0", Scale="1,1,1") {
    rotationMode=BMM_Additive
    rotationSpace=BCS_BoneSpace
    alphaBoolBlend {blendOption=Linear}
    alphaScaleBiasClamp {interpSpeedDecreasing=5, interpSpeedIncreasing=8}
    boneToModify {boneName=master_main_hair_jnt}
    rotation {yaw=20}
   }
   n26 = float * float(A=Get Head Back Rot(), B=Get Crouching() -> (Crouching:double)) {is=title K2Node_PromotableOperator}
   n27 = "Transform (Modify) Bone - Bone: master_main_hair_jnt"(ComponentPose=@n25, Alpha=@n26, Translation="-7,0,0", Rotation="-16,-5,21", Scale="1,1,1") {
    rotationMode=BMM_Additive
    rotationSpace=BCS_BoneSpace
    scaleSpace=BCS_WorldSpace
    translationMode=BMM_Additive
    translationSpace=BCS_BoneSpace
    alphaBoolBlend {blendOption=Linear}
    alphaScaleBiasClamp {interpSpeedDecreasing=5, interpSpeedIncreasing=8}
    boneToModify {boneName=master_main_hair_jnt}
    rotation=<rot -16 -5 21>
    translation {x=-7}
   }
   n28 = "Transform (Modify) Bone - Bone: master_R_hair_jnt"(ComponentPose=@n27, Alpha=@n26, Translation="-1.3,-2.4,0", Rotation="-15,-45,-12.5", Scale="1,1,1") {
    rotationMode=BMM_Additive
    rotationSpace=BCS_BoneSpace
    scaleSpace=BCS_WorldSpace
    translationMode=BMM_Additive
    translationSpace=BCS_BoneSpace
    alphaBoolBlend {blendOption=Linear}
    alphaScaleBiasClamp {interpSpeedDecreasing=5, interpSpeedIncreasing=8}
    boneToModify {boneName=master_R_hair_jnt}
    rotation=<rot -15 -45 -12.5>
    translation=<-1.3 -2.4>
   }
   n29 = float * float(A=Get Head R Rot(), B=Get Crouching())
   n30 = "Transform (Modify) Bone - Bone: master_main_hair_jnt"(ComponentPose=@n28, Alpha=@n29, Translation="0,-2.7,-4.4", Rotation="50,-20,0", Scale="1,1,1") {
    rotationMode=BMM_Additive
    rotationSpace=BCS_BoneSpace
    scaleSpace=BCS_WorldSpace
    translationMode=BMM_Additive
    translationSpace=BCS_BoneSpace
    alphaBoolBlend {blendOption=Linear}
    alphaScaleBiasClamp {interpSpeedDecreasing=5, interpSpeedIncreasing=8}
    boneToModify {boneName=master_main_hair_jnt}
    rotation {pitch=50, yaw=-20}
    translation {y=-2.7, z=-4.4}
   }
   n31 = "Transform (Modify) Bone - Bone: master_R_hair_jnt"(ComponentPose=@n30, Alpha=@n29, Translation="0,0,0", Rotation="0,-20,-40", Scale="1,1,1") {
    rotationMode=BMM_Additive
    rotationSpace=BCS_BoneSpace
    scaleSpace=BCS_WorldSpace
    translationMode=BMM_Additive
    translationSpace=BCS_BoneSpace
    alphaBoolBlend {blendOption=Linear}
    alphaScaleBiasClamp {interpSpeedDecreasing=5, interpSpeedIncreasing=8}
    boneToModify {boneName=master_R_hair_jnt}
    rotation {roll=-40, yaw=-20}
   }
   n32 = "Transform (Modify) Bone - Bone: master_hair_sim_back_jnt"(ComponentPose=@n31, Alpha=Get Rail Grinding() -> (Rail Grinding:double), Translation="0,0,0", Rotation="0,20,0", Scale="1,1,1") {
    rotationMode=BMM_Additive
    rotationSpace=BCS_ParentBoneSpace
    alphaBoolBlend {blendOption=Linear}
    boneToModify {boneName=master_hair_sim_back_jnt}
    rotation {yaw=20}
   }
   Save cached pose 'TransformPose'(Pose=Component To Local(ComponentPose=@n32)) {cacheName=TransformPose}
   Save cached pose 'StartPose'(Pose=Copy Pose From Mesh() {bUseAttachedParent=true}) {cacheName=StartPose}
  ]}
 EventGraph {nodes [
   n1 = Event BlueprintUpdateAnimation()
   n2 = CheckForBackbling()
   Set Is Wearing Backbling(execute=[@n1.then, @n7.then], Is Wearing Backbling=OR Boolean(A=@n2, B=@n2.bIsBackblingLong))
   > Set Is Wearing Cape(Is Wearing Cape=@n2.bIsBackblingACape)
   n3 = OR Boolean!(A=@n2, B=@n2.bIsBackblingACape, C:bool=@n2.bIsBackblingLong) -> (ReturnValue)
   > Set Is Wearing Backpack Cape(Is Wearing Backpack Cape:bool=@n3) -> (Output_Get:bool)
   > Set Alpha Backpack(Alpha Backpack:double=SelectFloat(A=1, B=0, bPickA=@n3)) -> (Output_Get:double)
   > n4 = Gravity Override State Function DEPRECATED(CharacterPartAnimInstance=Self-Reference(), GameTimeInSeconds=GetGameTimeInSeconds(), GravityOverrideParams=Get Gravity Override Params())
   > Set Gravity Override(Gravity Override=@n4)
   n5 = GetOwnerSkeletalMeshComponent()
   n6 = BreakRotator(InRot=BreakTransform(InTransform=GetDeltaTransformFromRefPose(self=@n5, BoneName=head, BaseName=spine_05)).Rotation)
   > Set Head Back Rot(Head Back Rot:double=MapRangeClamped(Value=@n6.Yaw, InRangeA=-10, InRangeB=-55)) -> (Output_Get:double)
   > Set Head R Rot(Head R Rot:double=MapRangeClamped(Value=@n6.Pitch, InRangeA=0, InRangeB=45)) -> (Output_Get:double)
   > Set Skydive Hanging(Skydive Hanging:double="To Float (Boolean)"(InBool=Get bIsParachuteOpened())) -> (Output_Get:double)
   > Set Crouching(Crouching:double="To Float (Boolean)"(InBool=Get bIsCrouching())) -> (Output_Get:double)
   > Set Rail Grinding(Rail Grinding:double="To Float (Boolean)"(InBool=Get bIsOnGrindRail())) -> (Output_Get:double)
   > Set Front End(Front End:bool=Get bIsInFrontEnd()) -> (Output_Get:bool)
   n7 = Event OnCharacterCustomizationLoaded()
  ]}
]
variables [
 Alpha Backpack {type=number}
 Crouching {type=number}
 Front End {type=boolean}
 Gravity Override {type=Vector}
 Gravity Override Params {type=GravityOverrideParamsStruct}
 Head Back Rot {type=number}
 Head R Rot {type=number}
 Is Wearing Backbling {type=boolean}
 Is Wearing Backpack Cape {type=boolean}
 Is Wearing Cape {type=boolean}
 Rail Grinding {type=number}
 Skydive Hanging {type=number}
]
